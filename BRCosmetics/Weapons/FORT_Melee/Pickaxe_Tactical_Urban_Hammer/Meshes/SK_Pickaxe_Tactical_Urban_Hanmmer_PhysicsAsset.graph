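# Physics asset for the Tactical Urban Hammer pickaxe skeletal mesh, a two-handed melee weapon cosmetic.
# It defines a single capsule body on the Melee_TwoHanded_Handle bone, offset along the bone and rotated to lie along the weapon.
# With no other bodies or constraints, the whole weapon is represented as one rigid collision shape rather than a jointed setup.
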PhysicsAsset /BRCosmetics/Weapons/FORT_Melee/Pickaxe_Tactical_Urban_Hammer/Meshes/SK_Pickaxe_Tactical_Urban_Hanmmer_PhysicsAsset
bodies {Melee_TwoHanded_Handle [{shapeType=Capsule, length=69.0273, radius=28.1092, center=<0.0659 0.2795 22.8587>, rotation=<rot -1.7335 -90 0.3031>}]}
constraints []
